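# Physics asset for the CataclysmTower male medium body, WithBackpack variant, for secondary motion of its jacket, collar and Musk bones.
# Capsules on torso, upper arms and thighs form the body; left front panels add boxes.
# Base bones anchor Dyn chains to spine and clavicles with mostly limited joints, though back-panel joints are locked except Back_R_03.
# Some neighboring panels are cross-linked by free constraints; the bag body has no constraint.
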PhysicsAsset /BRCosmetics/Characters/Player/Male/Medium/Bodies/M_MED_CataclysmTower/Meshes/M_MED_CataclysmTower_Physics_WithBackpack
bodies {
 Base_Musk [{shapeType=Capsule, length=0.505, radius=0.505, shapeName=Base_Musk_capsule, rotation=<rot 0 -90 0>}]
 Base_collar_L [{shapeType=Capsule, length=1, radius=1, shapeName=Base_collar_L_capsule}]
 Base_collar_R [{shapeType=Capsule, length=1, radius=1, shapeName=Base_collar_R_capsule}]
 Base_jacket_Back_L [{shapeType=Capsule, length=1, radius=1, shapeName=Base_jacket_Back_L_capsule}]
 Base_jacket_Back_R [{shapeType=Capsule, length=1, radius=1, shapeName=Base_jacket_Back_L_capsule}]
 Base_jacket_Side_L [{shapeType=Capsule, length=1, radius=1, shapeName=Base_jacket_Side_L_capsule}]
 Base_jacket_Side_R [{shapeType=Capsule, length=1, radius=1, shapeName=Base_jacket_Side_L_capsule}]
 Base_jacket_front_L [{shapeType=Capsule, length=1, radius=1, shapeName=Base_jacket_front_L_capsule}]
 Base_jacket_front_R [{shapeType=Capsule, length=1, radius=1, shapeName=Base_jacket_Side_L_capsule}]
 Dyn_bag_pelvice [{
   shapeType=Capsule
   length=2.9379
   radius=4.3735
   shapeName=Dyn_bag_pelvice_capsule
   center=<1.6126 0 0.8916>
   rotation=<rot -0.0001 0 -90.0001>
  }]
 Dyn_collar_L_01 [{shapeType=Capsule, length=11.4169, radius=2.0948, shapeName=Dyn_collar_L_01_capsule, center=<2.6541 0 0.5802>, rotation=<rot 0 0 90>}]
 Dyn_collar_L_02 [{shapeType=Capsule, length=3.8971, radius=2.4898, shapeName=Dyn_collar_L_02_capsule, center=<3.0803 0 0.5578>, rotation=<rot 0 0 90>}]
 Dyn_collar_R_01 [{shapeType=Capsule, length=9.0118, radius=2.1938, shapeName=Dyn_collar_L_01_capsule, center=<-2.2773 0 -1.6835>, rotation=<rot 0 0 -73.2639>}]
 Dyn_collar_R_02 [{shapeType=Capsule, length=4.8524, radius=2.9255, shapeName=Dyn_collar_L_02_capsule, center=<-3.0803 0 -1.1398>, rotation=<rot 0 0 -90>}]
 Dyn_jacket_Back_L_01 [{shapeType=Capsule, length=10, radius=6, shapeName=Dyn_jacket_Back_L_01_capsule, center=<2.1312 0 4.1211>, rotation=<rot 0 0 90>}]
 Dyn_jacket_Back_L_02 [{shapeType=Capsule, length=10, radius=6, shapeName=Dyn_jacket_Back_L_02_capsule, center=<2.1312 0 4.3225>, rotation=<rot 0 0 90>}]
 Dyn_jacket_Back_L_03 [{shapeType=Capsule, length=10, radius=6, shapeName=Dyn_jacket_Back_L_03_capsule, center=<2.1312 0 5.192>, rotation=<rot 0 0 90>}]
 Dyn_jacket_Back_R_01 [{shapeType=Capsule, length=10, radius=6, shapeName=Dyn_jacket_Back_L_01_capsule, center=<-2.1312 0 -4.1211>, rotation=<rot 0 0 -90>}]
 Dyn_jacket_Back_R_02 [{shapeType=Capsule, length=10, radius=6, shapeName=Dyn_jacket_Back_L_02_capsule, center=<-2.1312 0 -4.3225>, rotation=<rot 0 0 -90>}]
 Dyn_jacket_Back_R_03 [{shapeType=Capsule, length=10, radius=6, shapeName=Dyn_jacket_Back_L_03_capsule, center=<-2.1312 0 -5.192>, rotation=<rot 0 0 -90>}]
 Dyn_jacket_Side_L_01 [{shapeType=Capsule, length=5, radius=5, shapeName=Dyn_jacket_Side_L_01_capsule, center=<2.9732 0 4.3766>, rotation=<rot 0 0 90>}]
 Dyn_jacket_Side_L_02 [{shapeType=Capsule, length=5.7179, radius=4.2821, shapeName=Dyn_jacket_Side_L_02_capsule, center=<2.9732 0 3.8186>, rotation=<rot 0 0 90>}]
 Dyn_jacket_Side_L_03 [{shapeType=Capsule, length=5.7213, radius=4.2787, shapeName=Dyn_jacket_Side_L_03_capsule, center=<2.9732 0 3.0721>, rotation=<rot 0 0 90>}]
 Dyn_jacket_Side_R_01 [{shapeType=Capsule, length=10, radius=6, shapeName=Dyn_jacket_Side_L_01_capsule, center=<-2.9732 0 -4.7963>, rotation=<rot 0 0 -90>}]
 Dyn_jacket_Side_R_02 [{shapeType=Capsule, length=10, radius=6, shapeName=Dyn_jacket_Side_L_02_capsule, center=<-2.9732 0 -4.6116>, rotation=<rot 0 0 -90>}]
 Dyn_jacket_Side_R_03 [{shapeType=Capsule, length=10, radius=6, shapeName=Dyn_jacket_Side_L_03_capsule, center=<-1.8254 0 -3.4918>, rotation=<rot 0 0 -90>}]
 Dyn_jacket_front_L_01 [{shapeType=Capsule, length=1, radius=1}, {shapeType=Box, extentX=5.8412, extentY=9.3844, extentZ=1.9014}]
 Dyn_jacket_front_L_02 [{shapeType=Capsule, length=1, radius=1}, {shapeType=Box, extentX=5.8412, extentY=9.6493, extentZ=2.2068, rotation=<rot 0 -10.6371 0>}]
 Dyn_jacket_front_L_03 [
  {shapeType=Capsule, length=1.078, radius=1.2907, center=<5.7207 4.6202 0>, rotation=<rot -90 -1.3263 1.8836>}
  {shapeType=Box, extentX=5.8412, extentY=10.2834, extentZ=2.3998, center=<0.7765 -0.1561 0>, rotation=<rot 0.0001 -14.2342 0>}
 ]
 Dyn_jacket_front_R_01 [{shapeType=Capsule, length=7, radius=4, shapeName=Dyn_jacket_Side_L_02_capsule, center=<-4.7684 1.9754 -2.2342>, rotation=<rot 0 0 85>}]
 Dyn_jacket_front_R_02 [{
   shapeType=Capsule
   length=6.7153
   radius=4.2847
   shapeName=Dyn_jacket_front_L_02_capsule
   center=<-4.7684 1.9754 -2.2342>
   rotation=<rot 0.0001 -11.2499 85>
  }]
 Dyn_jacket_front_R_03 [{
   shapeType=Capsule
   length=7
   radius=4
   shapeName=Dyn_jacket_front_L_03_capsule
   center=<-4.7684 2.3334 -1.8572>
   rotation=<rot -1.9409 -11.083 76.5409>
  }]
 Musk_dyn [{
   shapeType=Capsule
   length=12.4271
   radius=5.3647
   shapeName=Musk_dyn_capsule
   center=<-4.7952 5.0684 0.2279>
   rotation=<rot 0.1413 -105.6844 -84.0818>
  }]
 clavicle_l [{shapeType=Capsule, length=11.8155, radius=5.984, shapeName=clavicle_l_capsule, center=<5.9083 0 0>, rotation=<rot -90 179.8416 -179.8419>}]
 clavicle_r [{shapeType=Capsule, length=11.9788, radius=5.8208, shapeName=clavicle_l_capsule, center=<-5.9083 0 0>, rotation=<rot -90 179.8416 0.1581>}]
 pelvis [{shapeType=Capsule, length=9, radius=10.6, shapeName=pelvis_capsule, center=<0 -0.9976 0>}]
 spine_01 [{shapeType=Capsule, length=8.3954, radius=10.6428, shapeName=spine_01_capsule}]
 spine_02 [{shapeType=Capsule, length=4.6798, radius=9.5815, shapeName=spine_02_capsule, center=<0 0.6238 0>}]
 spine_03 [
  {shapeType=Capsule, length=4.7339, radius=10.2079, shapeName=spine_03_capsule, center=<0 1.0695 0>}
  {shapeType=Capsule, length=12.68, radius=4, shapeName=spine_03_capsule, center=<2.2759 10.0567 0>}
 ]
 spine_04 [{shapeType=Capsule, length=3.9847, radius=10.7463, shapeName=spine_04_capsule, center=<0 1.1459 0>}]
 spine_05 [{shapeType=Capsule, length=13.1739, radius=8.0272, shapeName=spine_05_capsule}]
 thigh_l [
  {
   shapeType=Capsule
   length=14.1243
   radius=8.8102
   shapeName=thigh_r_capsule
   center=<-19.5919 -0.4275 0.4393>
   rotation=<rot 90 176.4237 4.2512>
  }
  {
   shapeType=Capsule
   length=8.0349
   radius=8.1426
   shapeName=thigh_l_capsule
   center=<-2.8039 0.568 0.0582>
   rotation=<rot -88.8109 -179.9998 -175.4915>
  }
 ]
 thigh_r [{shapeType=Capsule, length=20.7646, radius=9.2354, shapeName=thigh_r_capsule, center=<14.7847 0 0>, rotation=<rot 90 180 180>}]
 upperarm_l [{shapeType=Capsule, length=20.6654, radius=6.7598, shapeName=upperarm_l_capsule, center=<13.7727 0 0>, rotation=<rot -90 179.9407 -179.9407>}]
 upperarm_r [
  {
   shapeType=Capsule
   length=19.9751
   radius=7.4502
   shapeName=upperarm_l_capsule
   center=<-13.7727 1.1715 0>
   rotation=<rot -90 3.2195 176.4729>
  }
  {
   shapeType=Capsule
   length=0.7762
   radius=8.2762
   shapeName=upperarm_r_capsule
   center=<-18.3085 2.5041 -0.1936>
   rotation=<rot 0 0.0001 75.0759>
  }
 ]
}
constraints [
 {
  bone1Name=spine_01
  bone2Name=pelvis
  swing1LimitDegrees=45
  swing1Motion=Free
  swing2LimitDegrees=45
  swing2Motion=Free
  twistLimitDegrees=45
  twistMotion=Free
 }
 {
  bone1Name=spine_02
  bone2Name=spine_01
  swing1LimitDegrees=45
  swing1Motion=Free
  swing2LimitDegrees=45
  swing2Motion=Free
  twistLimitDegrees=45
  twistMotion=Free
 }
 {
  bone1Name=spine_03
  bone2Name=spine_02
  swing1LimitDegrees=45
  swing1Motion=Free
  swing2LimitDegrees=45
  swing2Motion=Free
  twistLimitDegrees=45
  twistMotion=Free
 }
 {
  bone1Name=spine_04
  bone2Name=spine_03
  swing1LimitDegrees=45
  swing1Motion=Free
  swing2LimitDegrees=45
  swing2Motion=Free
  twistLimitDegrees=45
  twistMotion=Free
 }
 {
  bone1Name=Base_jacket_Back_L
  bone2Name=spine_04
  swing1LimitDegrees=45
  swing1Motion=Free
  swing2LimitDegrees=45
  swing2Motion=Free
  twistLimitDegrees=45
  twistMotion=Free
 }
 {
  bone1Name=Dyn_jacket_Back_L_01
  bone2Name=Base_jacket_Back_L
  swing1LimitDegrees=20
  swing1Motion=Locked
  swing2LimitDegrees=20
  swing2Motion=Locked
  twistLimitDegrees=0
  twistMotion=Locked
 }
 {
  bone1Name=Dyn_jacket_Back_L_02
  bone2Name=Dyn_jacket_Back_L_01
  swing1LimitDegrees=45
  swing1Motion=Locked
  swing2LimitDegrees=5
  swing2Motion=Locked
  twistLimitDegrees=0
  twistMotion=Locked
 }
 {
  bone1Name=Dyn_jacket_Back_L_03
  bone2Name=Dyn_jacket_Back_L_02
  swing1LimitDegrees=3
  swing1Motion=Locked
  swing2LimitDegrees=3
  swing2Motion=Locked
  twistLimitDegrees=1
  twistMotion=Locked
 }
 {
  bone1Name=Base_jacket_Side_L
  bone2Name=spine_04
  swing1LimitDegrees=45
  swing1Motion=Free
  swing2LimitDegrees=45
  swing2Motion=Free
  twistLimitDegrees=45
  twistMotion=Free
 }
 {bone1Name=Dyn_jacket_Side_L_01, bone2Name=Base_jacket_Side_L, swing1LimitDegrees=10, swing2LimitDegrees=10, twistLimitDegrees=0}
 {bone1Name=Dyn_jacket_Side_L_02, bone2Name=Dyn_jacket_Side_L_01, swing1LimitDegrees=7, swing2LimitDegrees=7, twistLimitDegrees=0}
 {bone1Name=Dyn_jacket_Side_L_03, bone2Name=Dyn_jacket_Side_L_02, swing1LimitDegrees=5, swing2LimitDegrees=5, twistLimitDegrees=0}
 {
  bone1Name=spine_05
  bone2Name=spine_04
  swing1LimitDegrees=45
  swing1Motion=Free
  swing2LimitDegrees=45
  swing2Motion=Free
  twistLimitDegrees=45
  twistMotion=Free
 }
 {
  bone1Name=clavicle_l
  bone2Name=spine_05
  swing1LimitDegrees=45
  swing1Motion=Free
  swing2LimitDegrees=45
  swing2Motion=Free
  twistLimitDegrees=45
  twistMotion=Free
 }
 {
  bone1Name=Base_collar_L
  bone2Name=clavicle_l
  swing1LimitDegrees=45
  swing1Motion=Free
  swing2LimitDegrees=45
  swing2Motion=Free
  twistLimitDegrees=45
  twistMotion=Free
 }
 {bone1Name=Dyn_collar_L_01, bone2Name=Base_collar_L, swing1LimitDegrees=15, swing2LimitDegrees=15, twistLimitDegrees=10}
 {bone1Name=Dyn_collar_L_02, bone2Name=Dyn_collar_L_01, swing1LimitDegrees=10, swing2LimitDegrees=10, twistLimitDegrees=0}
 {
  bone1Name=upperarm_l
  bone2Name=clavicle_l
  swing1LimitDegrees=45
  swing1Motion=Free
  swing2LimitDegrees=45
  swing2Motion=Free
  twistLimitDegrees=45
  twistMotion=Free
 }
 {
  bone1Name=Base_collar_R
  bone2Name=clavicle_r
  swing1LimitDegrees=45
  swing1Motion=Free
  swing2LimitDegrees=45
  swing2Motion=Free
  twistLimitDegrees=45
  twistMotion=Free
 }
 {
  bone1Name=upperarm_r
  bone2Name=clavicle_r
  swing1LimitDegrees=45
  swing1Motion=Free
  swing2LimitDegrees=45
  swing2Motion=Free
  twistLimitDegrees=45
  twistMotion=Free
 }
 {bone1Name=Dyn_jacket_front_R_02, bone2Name=Dyn_jacket_front_R_01, swing1LimitDegrees=10, swing2LimitDegrees=20, twistLimitDegrees=0}
 {bone1Name=Dyn_jacket_front_R_03, bone2Name=Dyn_jacket_front_R_02, swing1LimitDegrees=5, swing2LimitDegrees=15, twistLimitDegrees=0}
 {
  bone1Name=clavicle_r
  bone2Name=spine_05
  swing1LimitDegrees=45
  swing1Motion=Free
  swing2LimitDegrees=45
  swing2Motion=Free
  twistLimitDegrees=45
  twistMotion=Free
 }
 {bone1Name=Dyn_collar_R_01, bone2Name=Base_collar_R, swing1LimitDegrees=5, swing2LimitDegrees=5, twistLimitDegrees=1}
 {bone1Name=Dyn_collar_R_02, bone2Name=Dyn_collar_R_01, swing1LimitDegrees=4, swing2LimitDegrees=4, twistLimitDegrees=0}
 {
  bone1Name=Base_jacket_Back_R
  bone2Name=spine_04
  swing1LimitDegrees=45
  swing1Motion=Free
  swing2LimitDegrees=45
  swing2Motion=Free
  twistLimitDegrees=45
  twistMotion=Free
 }
 {
  bone1Name=Dyn_jacket_Back_R_01
  bone2Name=Base_jacket_Back_R
  swing1LimitDegrees=20
  swing1Motion=Locked
  swing2LimitDegrees=20
  swing2Motion=Locked
  twistLimitDegrees=0
  twistMotion=Locked
 }
 {
  bone1Name=Dyn_jacket_Back_R_02
  bone2Name=Dyn_jacket_Back_R_01
  swing1LimitDegrees=45
  swing1Motion=Locked
  swing2LimitDegrees=5
  swing2Motion=Locked
  twistLimitDegrees=0
  twistMotion=Locked
 }
 {bone1Name=Dyn_jacket_Back_R_03, bone2Name=Dyn_jacket_Back_R_02, swing1LimitDegrees=3, swing2LimitDegrees=3, twistLimitDegrees=1}
 {
  bone1Name=Base_jacket_Side_R
  bone2Name=spine_04
  swing1LimitDegrees=45
  swing1Motion=Free
  swing2LimitDegrees=45
  swing2Motion=Free
  twistLimitDegrees=45
  twistMotion=Free
 }
 {bone1Name=Dyn_jacket_Side_R_02, bone2Name=Dyn_jacket_Side_R_01, swing1LimitDegrees=7, swing2LimitDegrees=7, twistLimitDegrees=0}
 {bone1Name=Base_jacket_front_R, bone2Name=spine_04, swing1LimitDegrees=7, swing2LimitDegrees=7, twistLimitDegrees=0}
 {bone1Name=Dyn_jacket_front_R_01, bone2Name=Base_jacket_front_R, swing1LimitDegrees=15, swing2LimitDegrees=25, twistLimitDegrees=0}
 {bone1Name=Dyn_jacket_Side_R_03, bone2Name=Dyn_jacket_Side_R_02, swing1LimitDegrees=5, swing2LimitDegrees=5, twistLimitDegrees=0}
 {bone1Name=Dyn_jacket_Side_R_01, bone2Name=Base_jacket_Side_R, swing1LimitDegrees=10, swing2LimitDegrees=10, twistLimitDegrees=0}
 {
  bone1Name=Base_jacket_front_L
  bone2Name=spine_04
  swing1LimitDegrees=45
  swing1Motion=Free
  swing2LimitDegrees=45
  swing2Motion=Free
  twistLimitDegrees=45
  twistMotion=Free
 }
 {bone1Name=Dyn_jacket_front_L_01, bone2Name=Base_jacket_front_L, swing1LimitDegrees=10, swing2LimitDegrees=10, twistLimitDegrees=10}
 {bone1Name=Dyn_jacket_front_L_02, bone2Name=Dyn_jacket_front_L_01, swing1LimitDegrees=15, swing2LimitDegrees=15, twistLimitDegrees=15}
 {bone1Name=Dyn_jacket_front_L_03, bone2Name=Dyn_jacket_front_L_02, swing1LimitDegrees=15, swing2LimitDegrees=20, twistLimitDegrees=15}
 {
  bone1Name=Dyn_jacket_Back_R_02
  bone2Name=Dyn_jacket_Back_L_02
  swing1LimitDegrees=45
  swing1Motion=Free
  swing2LimitDegrees=45
  swing2Motion=Free
  twistLimitDegrees=45
  twistMotion=Free
 }
 {
  bone1Name=Dyn_jacket_Side_L_03
  bone2Name=Dyn_jacket_Back_L_03
  swing1LimitDegrees=45
  swing1Motion=Free
  swing2LimitDegrees=45
  swing2Motion=Free
  twistLimitDegrees=45
  twistMotion=Free
 }
 {
  bone1Name=Dyn_jacket_front_R_03
  bone2Name=Dyn_jacket_Side_R_03
  swing1LimitDegrees=45
  swing1Motion=Free
  swing2LimitDegrees=45
  swing2Motion=Free
  twistLimitDegrees=45
  twistMotion=Free
 }
 {
  bone1Name=thigh_r
  bone2Name=pelvis
  swing1LimitDegrees=45
  swing1Motion=Free
  swing2LimitDegrees=45
  swing2Motion=Free
  twistLimitDegrees=45
  twistMotion=Free
 }
 {
  bone1Name=Dyn_jacket_Side_L_03
  bone2Name=Dyn_jacket_front_L_03
  swing1LimitDegrees=45
  swing1Motion=Free
  swing2LimitDegrees=45
  swing2Motion=Free
  twistLimitDegrees=45
  twistMotion=Free
 }
 {
  bone1Name=thigh_l
  bone2Name=pelvis
  swing1LimitDegrees=45
  swing1Motion=Free
  swing2LimitDegrees=45
  swing2Motion=Free
  twistLimitDegrees=45
  twistMotion=Free
 }
 {
  bone1Name=Base_Musk
  bone2Name=spine_05
  swing1LimitDegrees=45
  swing1Motion=Free
  swing2LimitDegrees=45
  swing2Motion=Free
  twistLimitDegrees=45
  twistMotion=Free
 }
 {bone1Name=Musk_dyn, bone2Name=Base_Musk, swing1LimitDegrees=12, swing2LimitDegrees=4, twistLimitDegrees=0}
]
