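# Post-process-style AnimBlueprint for the CataclysmTower male body cosmetic, adding secondary motion to jacket, "Musk", bag, pocket, zipper and shorts bones.
# The parent mesh pose passes through a constraint and Control Rig into a cache feeding two blended RigidBody branches (backpack and crouch physics assets).
# AnimDynamics, bone constraints and a look-at then layer on accessory motion.
# Each update sets state flags and derives gravity/RBAN settings through a deprecated helper.
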AnimBlueprint /BRCosmetics/Characters/Player/Male/Medium/Bodies/M_MED_CataclysmTower/Meshes/M_MED_CataclysmTower_AnimBP
target_skeleton=&/M_MED_CataclysmTower_Skeleton
graphs [
 AnimGraph {nodes [
   Anim Dynamics - Bone: Dyn_bag_pelvice(Alpha=1, ExternalForce="0,0,0") {
    angularDampingOverride=1
    angularSpringConstant=800
    bAngularSpring=true
    bOverrideAngularDamping=true
    bUseGravityOverride=true
    simulationSpace=RootRelative
    alphaBoolBlend {blendOption=Linear}
    boundBone {boneName=Dyn_bag_pelvice}
    componentAppliedLinearAccClamp=<10000 10000 10000>
    componentLinearAccScale=<2 2 2>
    componentLinearVelScale=<2 2 2>
    gravityOverride {z=-980}
    physicsBodyDefinitions [{
      boundBone {boneName=Dyn_bag_pelvice}
      boxExtents=<10 10 10>
      constraintSetup {angularLimitsMax=<8 10 0>, angularLimitsMin=<-8 -5 0>, angularTarget=<1 0 0>}
      localJointOffset=<0 0 -0.5>
     }]
   }
   n1 = RigidBody!(
    ComponentPose=Local To Component(LocalPose=Use cached pose 'init'())
    Alpha
    OverrideWorldGravity:Vector="0,0,-980"
    ExternalForce
    SimSpaceSettings:Sim Space Settings Structure=(WorldAlpha=0,VelocityScaleZ=1,DampingAlpha=1,MaxLinearVelocity=10000,MaxAngularVelocity=10000,MaxLinearAcceleration=10000,MaxAngularAcceleration=10000,ExternalLinearDragV=(X=0,Y=0,Z=0),ExternalLinearVelocity=(X=0,Y=0,Z=0),ExternalAngularVelocity=(X=0,Y=0,Z=0))
   ) -> (Pose) {
    bClampLinearTranslationLimitToRefPose=true
    bOverrideWorldGravity=true
    overridePhysicsAsset=&/M_MED_CataclysmTower_Physics_WithBackpack
    simulationSpace=BaseBoneSpace
    alphaBoolBlend {blendOption=Linear}
    baseBoneRef {boneName=spine_03}
    componentAppliedLinearAccClamp=<5000 5000 5000>
    overrideWorldGravity {z=-980}
   }
   n2 = "Transform (Modify) Bone - Bone: Dyn_jacket_Back_L_01"(ComponentPose=@n1, Alpha=1, Translation="0,0,0", Rotation="0,0,-18", Scale="1,1,1") {rotationMode=BMM_Additive, alphaBoolBlend {blendOption=Linear}, boneToModify {boneName=Dyn_jacket_Back_L_01}, rotation {roll=-18}}
   n3 = Component To Local(ComponentPose=@n2)
   n4 = RigidBody(
    ComponentPose=Local To Component(LocalPose=Use cached pose 'init'())
    OverrideWorldGravity="0,0,-980"
    SimSpaceSettings=(WorldAlpha=0,VelocityScaleZ=1,DampingAlpha=1,MaxLinearVelocity=10000,MaxAngularVelocity=10000,MaxLinearAcceleration=10000,MaxAngularAcceleration=10000,ExternalLinearDragV=(X=0,Y=0,Z=0),ExternalLinearVelocity=(X=0,Y=0,Z=0),ExternalAngularVelocity=(X=0,Y=0,Z=0))
   ) {
    bClampLinearTranslationLimitToRefPose=true
    bOverrideWorldGravity=true
    overridePhysicsAsset=&/M_MED_CataclysmTower_Physics_crouch
    simulationSpace=BaseBoneSpace
    alphaBoolBlend {blendOption=Linear}
    baseBoneRef {boneName=spine_03}
    componentAppliedLinearAccClamp=<5000 5000 5000>
    overrideWorldGravity {z=-980}
   }
   n5 = Component To Local(ComponentPose=@n4)
   n6 = "Transform (Modify) Bone - Bone: Base_Musk"!(
    ComponentPose=Local To Component(LocalPose=Blend Poses by bool(BlendPose_0=@n3, BlendPose_1=@n5) {blendType=Linear})
    bAlphaBoolEnabled:bool=True
    Translation="0,0,-0.5"
    Rotation="-8.308,9.145,-2.967"
   ) -> (Pose) {
    alphaInputType=Bool
    rotationMode=BMM_Additive
    translationMode=BMM_Additive
    alphaBoolBlend {blendOption=Linear}
    boneToModify {boneName=Base_Musk}
    rotation=<rot -8.308 9.145 -2.967>
    translation {z=-0.5}
   }
   n7 = "Transform (Modify) Bone - Bone: Musk_dyn"!(ComponentPose=@n6, bAlphaBoolEnabled:bool=True, Translation="0,0,0", Rotation="-7.312,-7.548,-11.126", Scale="1,1,1") -> (Pose) {
    alphaInputType=Bool
    rotationMode=BMM_Replace
    rotationSpace=BCS_ParentBoneSpace
    alphaBoolBlend {blendOption=Linear}
    boneToModify {boneName=Musk_dyn}
    rotation=<rot -7.312 -7.548 -11.126>
   }
   n8 = Anim Dynamics - Bone: Musk_dyn!(ComponentPose=@n7, bAlphaBoolEnabled:bool=True, ExternalForce="0,0,0") -> (Pose) {
    alphaInputType=Bool
    angularDampingOverride=0.999
    angularSpringConstant=1000
    bAngularSpring=true
    bOverrideAngularDamping=true
    simulationSpace=RootRelative
    alphaBoolBlend {blendOption=Linear}
    boundBone {boneName=Musk_dyn}
    componentAppliedLinearAccClamp=<3000 3000 3000>
    componentLinearAccScale=<0.3 0.3 0.3>
    componentLinearVelScale=<0.3 0.3 0.3>
    physicsBodyDefinitions [{
      boundBone {boneName=Musk_dyn}
      boxExtents=<7 7 7>
      constraintSetup {angularLimitsMax=<0 8 0>, angularLimitsMin=<0 -8 -5>, angularTarget=<1 0 0>}
      localJointOffset=<-0.2 0 0>
     }]
    relativeSpaceBone {boneName=spine_05}
   }
   n9 = Anim Dynamics - Bone: Dyn_bag_R(ComponentPose=@n8, Alpha=1, ExternalForce="0,0,0") {
    angularDampingOverride=1
    angularSpringConstant=800
    bAngularSpring=true
    bOverrideAngularDamping=true
    bUseGravityOverride=true
    simulationSpace=RootRelative
    alphaBoolBlend {blendOption=Linear}
    boundBone {boneName=Dyn_bag_R}
    componentAppliedLinearAccClamp=<10000 10000 10000>
    componentLinearAccScale=<1 1 1>
    componentLinearVelScale=<1 1 1>
    gravityOverride {z=-980}
    physicsBodyDefinitions [{
      boundBone {boneName=Dyn_bag_R}
      boxExtents=<21 21 21>
      constraintSetup {angularLimitsMax=<10 15 5>, angularLimitsMin=<-10 -5 -5>, angularTarget=<1 0 0>}
      localJointOffset=<0 0 -0.7>
     }]
   }
   n10 = Anim Dynamics - Bone: pocket_dyn(ComponentPose=@n9, Alpha=1, ExternalForce="0,0,0") {
    angularDampingOverride=1
    angularSpringConstant=800
    bAngularSpring=true
    bOverrideAngularDamping=true
    bUseGravityOverride=true
    simulationSpace=RootRelative
    alphaBoolBlend {blendOption=Linear}
    boundBone {boneName=pocket_dyn}
    componentAppliedLinearAccClamp=<10000 10000 10000>
    componentLinearAccScale=<1 1 1>
    componentLinearVelScale=<1 1 1>
    gravityOverride {z=-980}
    physicsBodyDefinitions [{
      boundBone {boneName=pocket_dyn}
      boxExtents=<10 10 10>
      constraintSetup {angularLimitsMax=<1 5 5>, angularLimitsMin=<-1 -40 -5>, angularTarget=<1 0 0>}
      localJointOffset=<0.5 0 0>
     }]
   }
   n11 = "Constraint - Dyn_bag_R (PPPPP)"!(
    ComponentPose=@n10
    Alpha=1
    ConstraintWeights_0=0.1199
    ConstraintWeights_1:float=0.0228
    ConstraintWeights_2:float=0.7529
    ConstraintWeights_3:float=0.0907
    ConstraintWeights_4:float=0.0136
   ) -> (Pose) {
    alphaBoolBlend {blendOption=Linear}
    boneToModify {boneName=Dyn_bag_R}
    constraintSetup [
     {targetBone {boneName=pelvis}}
     {targetBone {boneName=thigh_r}}
     {targetBone {boneName=thigh_twist_01_r}}
     {targetBone {boneName=deform_glute_r}}
     {targetBone {boneName=deform_groin_r}}
    ]
    constraintWeights [1,1,1,1,1]
   }
   n12 = "Constraint - pocket_dyn (PPPPPPP)"!(
    ComponentPose=@n11
    Alpha=1
    ConstraintWeights_0=0.0134
    ConstraintWeights_1:float=0.3536
    ConstraintWeights_2:float=0.3169
    ConstraintWeights_3:float=0.0011
    ConstraintWeights_4:float=0.0073
    ConstraintWeights_5:float=0.0162
    ConstraintWeights_6:float=0.2916
   ) -> (Pose) {
    alphaBoolBlend {blendOption=Linear}
    boneToModify {boneName=pocket_dyn}
    constraintSetup [
     {targetBone {boneName=clavicle_l}}
     {targetBone {boneName=upperarm_l}}
     {targetBone {boneName=upperarm_twist_01_l}}
     {targetBone {boneName=deform_upperarm_bk_base_l}}
     {targetBone {boneName=deform_upperarm_fr_base_l}}
     {targetBone {boneName=deform_upperarm_up_base_l}}
     {targetBone {boneName=upperarm_twist_02_l}}
    ]
    constraintWeights [1,1,1,1,1,1,1]
   }
   n13 = Look At - Bone: zipper_r_Base(ComponentPose=@n12, Alpha=1) {
    alphaBoolBlend {blendOption=Linear}
    boneToModify {boneName=zipper_r_Base}
    lookAtLocation=<-3.8683 -9.765 -8.8374>
    lookAtTarget {boneReference {boneName=Dyn_jacket_Side_R_03}}
   }
   n14 = Anim Dynamics - Bone: zipper_l_dyn(ComponentPose=@n13, Alpha=1, ExternalForce="0,0,0") {
    angularDampingOverride=0.5
    angularSpringConstant=400
    bAngularSpring=true
    bOverrideAngularDamping=true
    bUseGravityOverride=true
    simulationSpace=RootRelative
    alphaBoolBlend {blendOption=Linear}
    boundBone {boneName=zipper_l_dyn}
    componentAppliedLinearAccClamp=<10000 10000 10000>
    componentLinearAccScale=<1 1 1>
    componentLinearVelScale=<1 1 1>
    gravityOverride {z=-980}
    physicsBodyDefinitions [{
      boundBone {boneName=zipper_l_dyn}
      boxExtents=<25 25 25>
      constraintSetup {angularLimitsMax=<1 0 1>, angularLimitsMin=<-1 -50 -15>, angularTarget=<1 0 0>}
      localJointOffset=<0.6 0 0>
     }]
   }
   n15 = Anim Dynamics - Bone: zipper_r_dyn(ComponentPose=@n14, Alpha=1, ExternalForce="0,0,0") {
    angularDampingOverride=0.5
    angularSpringConstant=400
    bAngularSpring=true
    bOverrideAngularDamping=true
    bUseGravityOverride=true
    simulationSpace=RootRelative
    alphaBoolBlend {blendOption=Linear}
    boundBone {boneName=zipper_r_dyn}
    componentAppliedLinearAccClamp=<10000 10000 10000>
    componentLinearAccScale=<1 1 1>
    componentLinearVelScale=<1 1 1>
    gravityOverride {z=-980}
    physicsBodyDefinitions [{
      boundBone {boneName=zipper_r_dyn}
      boxExtents=<25 25 25>
      constraintSetup {angularLimitsMax=<1 2 15>, angularLimitsMin=<-1 -50 -15>, angularTarget=<1 0 0>}
      localJointOffset=<0.6 0 0>
     }]
   }
   n16 = "Constraint - zipper_front_Base (PPPPP)"!(
    ComponentPose=@n15
    Alpha=1
    ConstraintWeights_0=0.0001
    ConstraintWeights_1:float=0.033
    ConstraintWeights_2:float=0.9663
    ConstraintWeights_3:float=0.0001
    ConstraintWeights_4:float=0.0005
   ) -> (Pose) {
    alphaBoolBlend {blendOption=Linear}
    boneToModify {boneName=zipper_front_Base}
    constraintSetup [
     {targetBone {boneName=Dyn_jacket_front_L_01}}
     {targetBone {boneName=Dyn_jacket_front_L_02}}
     {targetBone {boneName=Dyn_jacket_front_L_03}}
     {targetBone {boneName=Dyn_jacket_Side_L_02}}
     {targetBone {boneName=Dyn_jacket_Side_L_03}}
    ]
    constraintWeights [1,1,1,1,1]
   }
   n17 = Anim Dynamics - Bone: zipper_front_dyn!(ComponentPose=@n16, Alpha=1, GravityOverride:Vector="0,0,-980", ExternalForce="0,0,0") -> (Pose) {
    angularDampingOverride=0.99
    angularSpringConstant=90
    bAngularSpring=true
    bOverrideAngularDamping=true
    bUseGravityOverride=true
    gravityScale=0.35
    simulationSpace=BoneRelative
    alphaBoolBlend {blendOption=Linear}
    boundBone {boneName=zipper_front_dyn}
    componentAppliedLinearAccClamp=<3000 3000 3000>
    componentLinearAccScale=<1 1 1>
    componentLinearVelScale=<1 1 1>
    physicsBodyDefinitions [{
      boundBone {boneName=zipper_front_dyn}
      boxExtents=<10 5 10>
      constraintSetup {coneAngle=1, angularLimitsMax=<5 5 40>, angularLimitsMin=<-5 -90 -15>, angularTarget=<1 0 0>}
      localJointOffset=<0.4 0 0>
     }]
    relativeSpaceBone {boneName=root}
   }
   n18 = Anim Dynamics - Bone: pocket_dyn!(ComponentPose=@n17, Alpha=1, GravityOverride:Vector="0,0,-980", ExternalForce="0,0,0") -> (Pose) {
    angularDampingOverride=1
    angularSpringConstant=500
    bAngularSpring=true
    bOverrideAngularDamping=true
    bUseGravityOverride=true
    gravityScale=0.35
    simulationSpace=RootRelative
    alphaBoolBlend {blendOption=Linear}
    boundBone {boneName=pocket_dyn}
    componentAppliedLinearAccClamp=<6000 6000 6000>
    componentLinearAccScale=<1 1 1>
    componentLinearVelScale=<1 1 1>
    gravityOverride {z=-980}
    physicsBodyDefinitions [{
      boundBone {boneName=pocket_dyn}
      boxExtents=<10 10 10>
      constraintSetup {angularLimitsMax=<5 15 20>, angularLimitsMin=<-5 -60 -20>, angularTarget=<1 0 0>}
      localJointOffset=<1.2 0 0>
     }]
   }
   n19 = "Constraint - zipper_r_Base (PPPPPPPP)"!(
    ComponentPose=@n18
    Alpha=1
    ConstraintWeights_0=0.0819
    ConstraintWeights_1:float=0.1052
    ConstraintWeights_2:float=0.0132
    ConstraintWeights_3:float=0.3915
    ConstraintWeights_4:float=0.3178
    ConstraintWeights_5:float=0.0869
    ConstraintWeights_6:float=0.0036
    ConstraintWeights_7:float=0
   ) -> (Pose) {
    alphaBoolBlend {blendOption=Linear}
    boneToModify {boneName=zipper_r_Base}
    constraintSetup [
     {targetBone {boneName=spine_03}}
     {targetBone {boneName=Dyn_jacket_front_R_01}}
     {targetBone {boneName=Dyn_jacket_front_R_02}}
     {targetBone {boneName=Dyn_jacket_Side_R_01}}
     {targetBone {boneName=Dyn_jacket_Side_R_02}}
     {targetBone {boneName=Dyn_jacket_Side_R_03}}
     {targetBone {boneName=spine_04}}
     {targetBone {boneName=deform_pec_base_r}}
    ]
    constraintWeights [1,1,1,1,1,1,1,1]
   }
   n20 = "Constraint - zipper_l_Base (PPPPP)"!(
    ComponentPose=@n19
    Alpha=1
    ConstraintWeights_0=0.1449
    ConstraintWeights_1:float=0.1597
    ConstraintWeights_2:float=0.3119
    ConstraintWeights_3:float=0.2852
    ConstraintWeights_4:float=0.0984
   ) -> (Pose) {
    alphaBoolBlend {blendOption=Linear}
    boneToModify {boneName=zipper_l_Base}
    constraintSetup [
     {targetBone {boneName=Dyn_jacket_front_L_01}}
     {targetBone {boneName=Dyn_jacket_front_L_02}}
     {targetBone {boneName=Dyn_jacket_Side_L_01}}
     {targetBone {boneName=Dyn_jacket_Side_L_02}}
     {targetBone {boneName=Dyn_jacket_Side_L_03}}
    ]
    constraintWeights [1,1,1,1,1]
   }
   n21 = Anim Dynamics - Bone: Dyn_shorts_L_front(ComponentPose=@n20, Alpha=1, ExternalForce="0,0,0") {
    angularDampingOverride=1
    angularSpringConstant=800
    bAngularSpring=true
    bOverrideAngularDamping=true
    bUseGravityOverride=true
    simulationSpace=RootRelative
    alphaBoolBlend {blendOption=Linear}
    boundBone {boneName=Dyn_shorts_L_front}
    componentAppliedLinearAccClamp=<5000 5000 5000>
    componentLinearAccScale=<0.65 0.65 0.65>
    componentLinearVelScale=<0.65 0.65 0.65>
    gravityOverride {z=-980}
    physicsBodyDefinitions [{
      boundBone {boneName=Dyn_shorts_L_front}
      boxExtents=<10 10 10>
      constraintSetup {coneAngle=1, angularLimitsMax=<1 20 15>, angularLimitsMin=<-1 -20 -15>, angularTarget=<1 0 0>}
      localJointOffset=<0.5 0 0>
     }]
    relativeSpaceBone {boneName=thigh_l}
   }
   n22 = Anim Dynamics - Bone: Dyn_shorts_L_back(ComponentPose=@n21, Alpha=1, ExternalForce="0,0,0") {
    angularDampingOverride=1
    angularSpringConstant=800
    bAngularSpring=true
    bOverrideAngularDamping=true
    bUseGravityOverride=true
    simulationSpace=RootRelative
    alphaBoolBlend {blendOption=Linear}
    boundBone {boneName=Dyn_shorts_L_back}
    componentAppliedLinearAccClamp=<10000 10000 10000>
    componentLinearAccScale=<1 1 1>
    componentLinearVelScale=<1 1 1>
    gravityOverride {z=-980}
    physicsBodyDefinitions [{
      boundBone {boneName=Dyn_shorts_L_back}
      boxExtents=<10 10 10>
      constraintSetup {angularLimitsMax=<1 2 10>, angularLimitsMin=<-1 -15 -10>, angularTarget=<1 0 0>}
      localJointOffset=<1 0 0>
     }]
   }
   n23 = Anim Dynamics - Bone: Dyn_shorts_L_Side_in(ComponentPose=@n22, Alpha=1, ExternalForce="0,0,0") {
    angularDampingOverride=1
    angularSpringConstant=800
    bAngularSpring=true
    bOverrideAngularDamping=true
    bUseGravityOverride=true
    simulationSpace=RootRelative
    alphaBoolBlend {blendOption=Linear}
    boundBone {boneName=Dyn_shorts_L_Side_in}
    componentAppliedLinearAccClamp=<10000 10000 10000>
    componentLinearAccScale=<1 1 1>
    componentLinearVelScale=<1 1 1>
    gravityOverride {z=-980}
    physicsBodyDefinitions [{
      boundBone {boneName=Dyn_shorts_L_Side_in}
      boxExtents=<10 10 10>
      constraintSetup {angularLimitsMax=<1 2 10>, angularLimitsMin=<-1 -15 -10>, angularTarget=<1 0 0>}
      localJointOffset=<1 0 0>
     }]
   }
   n24 = Anim Dynamics - Bone: Dyn_shorts_L_Side_out(ComponentPose=@n23, Alpha=1, ExternalForce="0,0,0") {
    angularDampingOverride=1
    angularSpringConstant=800
    bAngularSpring=true
    bOverrideAngularDamping=true
    bUseGravityOverride=true
    simulationSpace=RootRelative
    alphaBoolBlend {blendOption=Linear}
    boundBone {boneName=Dyn_shorts_L_Side_out}
    componentAppliedLinearAccClamp=<10000 10000 10000>
    componentLinearAccScale=<1 1 1>
    componentLinearVelScale=<1 1 1>
    gravityOverride {z=-980}
    physicsBodyDefinitions [{
      boundBone {boneName=Dyn_shorts_L_Side_out}
      boxExtents=<10 10 10>
      constraintSetup {angularLimitsMax=<1 2 10>, angularLimitsMin=<-1 -15 -10>, angularTarget=<1 0 0>}
      localJointOffset=<1 0 0>
     }]
   }
   n25 = Anim Dynamics - Bone: Dyn_shorts_R_front(ComponentPose=@n24, Alpha=1, ExternalForce="0,0,0") {
    angularDampingOverride=1
    angularSpringConstant=800
    bAngularSpring=true
    bOverrideAngularDamping=true
    bUseGravityOverride=true
    simulationSpace=RootRelative
    alphaBoolBlend {blendOption=Linear}
    boundBone {boneName=Dyn_shorts_R_front}
    componentAppliedLinearAccClamp=<5000 5000 5000>
    componentLinearAccScale=<0.65 0.65 0.65>
    componentLinearVelScale=<0.65 0.65 0.65>
    gravityOverride {z=-980}
    physicsBodyDefinitions [{
      boundBone {boneName=Dyn_shorts_R_front}
      boxExtents=<10 10 10>
      constraintSetup {angularLimitsMax=<1 20 15>, angularLimitsMin=<-1 -20 -15>, angularTarget=<1 0 0>}
      localJointOffset=<-0.5 0 0>
     }]
    relativeSpaceBone {boneName=root}
   }
   n26 = Anim Dynamics - Bone: Dyn_shorts_R_back(ComponentPose=@n25, Alpha=1, ExternalForce="0,0,0") {
    angularDampingOverride=1
    angularSpringConstant=800
    bAngularSpring=true
    bOverrideAngularDamping=true
    bUseGravityOverride=true
    simulationSpace=RootRelative
    alphaBoolBlend {blendOption=Linear}
    boundBone {boneName=Dyn_shorts_R_back}
    componentAppliedLinearAccClamp=<10000 10000 10000>
    componentLinearAccScale=<1 1 1>
    componentLinearVelScale=<1 1 1>
    gravityOverride {z=-980}
    physicsBodyDefinitions [{
      boundBone {boneName=Dyn_shorts_R_back}
      boxExtents=<10 10 10>
      constraintSetup {angularLimitsMax=<1 2 10>, angularLimitsMin=<-1 -15 -10>, angularTarget=<1 0 0>}
      localJointOffset=<-1 0 0>
     }]
   }
   n27 = Anim Dynamics - Bone: Dyn_shorts_R_Side_out(ComponentPose=@n26, Alpha=1, ExternalForce="0,0,0") {
    angularDampingOverride=1
    angularSpringConstant=800
    bAngularSpring=true
    bOverrideAngularDamping=true
    bUseGravityOverride=true
    simulationSpace=RootRelative
    alphaBoolBlend {blendOption=Linear}
    boundBone {boneName=Dyn_shorts_R_Side_out}
    componentAppliedLinearAccClamp=<10000 10000 10000>
    componentLinearAccScale=<1 1 1>
    componentLinearVelScale=<1 1 1>
    gravityOverride {z=-980}
    physicsBodyDefinitions [{
      boundBone {boneName=Dyn_shorts_R_Side_out}
      boxExtents=<10 10 10>
      constraintSetup {angularLimitsMax=<1 2 10>, angularLimitsMin=<-1 -15 -10>, angularTarget=<1 0 0>}
      localJointOffset=<-1 0.5 0>
     }]
   }
   n28 = Anim Dynamics - Bone: Dyn_shorts_R_Side_in(ComponentPose=@n27, Alpha=1, ExternalForce="0,0,0") {
    angularDampingOverride=1
    angularSpringConstant=800
    bAngularSpring=true
    bOverrideAngularDamping=true
    bUseGravityOverride=true
    simulationSpace=RootRelative
    alphaBoolBlend {blendOption=Linear}
    boundBone {boneName=Dyn_shorts_R_Side_in}
    componentAppliedLinearAccClamp=<10000 10000 10000>
    componentLinearAccScale=<1 1 1>
    componentLinearVelScale=<1 1 1>
    gravityOverride {z=-980}
    physicsBodyDefinitions [{
      boundBone {boneName=Dyn_shorts_R_Side_in}
      boxExtents=<10 10 10>
      constraintSetup {angularLimitsMax=<1 2 10>, angularLimitsMin=<-1 -15 -10>, angularTarget=<1 0 0>}
      localJointOffset=<-1 0 0>
     }]
   }
   Output Pose(Result=Component To Local(ComponentPose=@n28)) {layerGroup=None}
   n29 = "Constraint - Base_Musk (PPPPPPPP)"!(
    ComponentPose=Local To Component(LocalPose=Copy Pose From Mesh() {bUseAttachedParent=true})
    bAlphaBoolEnabled:bool=True
    ConstraintWeights_0=0.0119
    ConstraintWeights_1:float=0.0523
    ConstraintWeights_2:float=0.1962
    ConstraintWeights_3:float=0.614
    ConstraintWeights_4:float=0.0662
    ConstraintWeights_5:float=0.0168
    ConstraintWeights_6:float=0.039
    ConstraintWeights_7:float=0.0037
   ) -> (Pose) {
    alphaInputType=Bool
    alphaBoolBlend {blendOption=Linear}
    boneToModify {boneName=Base_Musk}
    constraintSetup [
     {targetBone {boneName=spine_02}}
     {targetBone {boneName=spine_03}}
     {targetBone {boneName=spine_04}}
     {targetBone {boneName=spine_05}}
     {targetBone {boneName=clavicle_l}}
     {targetBone {boneName=neck_01}}
     {targetBone {boneName=neck_02}}
     {targetBone {boneName=deform_pec_l}}
    ]
    constraintWeights [1,1,1,1,1,1,1,1]
   }
   n30 = Control Rig!(
    Source=Component To Local(ComponentPose=@n29)
    Alpha
    Paying emote:bool=Get bIsPlayingEmote()
    isSkydive:bool=Get bIsSkyDiving()
    isParachuting:bool=Get isParachute() -> (isParachute:bool)
    isALT:bool=Get isALTsskin() -> (isALTsskin:bool)
    isLOD2:bool=Get isALTsskin()
   ) -> (Pose) {
    alphaBoolBlend {blendOption=Linear}
    controlRigAssetReference {blueprintRigClass=&/M_MED_CataclysmTower_CtrlRig.M_MED_CataclysmTower_CtrlRig_C}
   }
   Save cached pose 'init'(Pose=@n30) {cacheName=init}
  ]}
 EventGraph {nodes [
   Event BlueprintUpdateAnimation()
   > n1 = Sequence()
   > n2 = Gravity Override State Function DEPRECATED(
    CharacterPartAnimInstance=Self-Reference()
    GameTimeInSeconds=GetGameTimeInSeconds()
    GravityOverrideParams=Get Gravity Override Params()
    RBAN Params=Get RBAN Params()
   )
   > Set Gravity Override(Gravity Override=@n2)
   > n3 = Set RBAN Settings Out(RBAN Settings Out:RBANSetting Struct Structure=@n2.RBAN SettingsOut) -> (Output_Get:RBANSetting Struct Structure)
   > Set isDiveMode(isDiveMode:bool=NOT Boolean(A=Get bIsSkydiveDiveMode())) -> (Output_Get:bool)
   > Set isALTsskin(isALTsskin:bool=NOT Boolean(A=Get bRigidBodyEnabled())) -> (Output_Get:bool)
   > Set is LOD2(is LOD2:bool="integer >= integer"(A=Get CurrentLOD(), B=2)) -> (Output_Get:bool)
   > n4 = Set isWearingBackpack(isWearingBackpack:bool=CheckForBackbling()) -> (Output_Get:bool)
   > Set isSkyDivingAndBackpack(isSkyDivingAndBackpack:bool=AND Boolean(A=@n4, B=Get bIsSkyDiving())) -> (Output_Get:bool)
   > Set Is NOT Playing Full Body Slot In Front End(Is NOT Playing Full Body Slot In Front End:bool=NOT Boolean(A=Get bIsPlayingFullBodySlotInFrontEnd() -> (bIsPlayingFullBodySlotInFrontEnd:bool))) -> (Output_Get:bool)
   Set frontendAlpha(execute=@n1.then_1, frontendAlpha:double=SelectFloat(A=0.65, B=1, bPickA=Get bIsInFrontEnd())) -> (Output_Get:double)
   Break Sim Space Settings!(SimSpaceSettings:"Sim Space Settings Structure (by ref)"=Break RBANSetting Struct!(RBANSettingStruct:"RBANSetting Struct Structure (by ref)"=@n3) -> (
     ComponentLinearAccScale_28_#1:Vector=~
     ComponentLinearVelScale_31_#2:Vector=~
     SimSpaceSettings_34_#3:Sim Space Settings Structure
    ) {is=title K2Node_BreakStruct}.SimSpaceSettings_34_#3) -> (
    WorldAlpha:float=~
    VelocityScaleZ:float=~
    DampingAlpha:float=~
    MaxLinearVelocity:float=~
    MaxAngularVelocity:float=~
    MaxLinearAcceleration:float=~
    MaxAngularAcceleration:float=~
    ExternalLinearDragV:Vector=~
    ExternalLinearVelocity:Vector=~
    ExternalAngularVelocity:Vector=~
   ) {is=title K2Node_BreakStruct}
  ]}
]
variables [
 frontendAlpha {type=number}
 Gravity Override {type=Vector}
 Gravity Override Params {type=GravityOverrideParamsStruct}
 is LOD2 {type=boolean}
 Is NOT Playing Full Body Slot In Front End {type=boolean}
 isALTsskin {type=boolean}
 isDiveMode {type=boolean}
 isParachute {type=boolean}
 isSkyDivingAndBackpack {type=boolean}
 isWearingBackpack {type=boolean}
 RBAN Params {
  type=RBANParamsStruct
  default {
   dBNO {simSpaceSettings {
     maxAngularAcceleration=1000
     maxAngularVelocity=3
     maxLinearAcceleration=300
     maxLinearVelocity=700
     velocityScaleZ=0
     externalLinearDragV=<0 0.5 0>
    }}
   dBNO_Carried {simSpaceSettings {
     maxAngularAcceleration=1000
     maxAngularVelocity=3
     maxLinearAcceleration=300
     maxLinearVelocity=700
     velocityScaleZ=0
     externalLinearDragV=<0 0.5 0>
    }}
   emoteOrMelee {simSpaceSettings {
     maxAngularAcceleration=1000
     maxAngularVelocity=3
     maxLinearAcceleration=300
     maxLinearVelocity=700
     velocityScaleZ=0
     externalLinearDragV=<0 0.5 0>
    }}
   front_End {simSpaceSettings {
     maxAngularAcceleration=1000
     maxAngularVelocity=3
     maxLinearAcceleration=300
     maxLinearVelocity=700
     velocityScaleZ=0
     externalLinearDragV=<0 0.5 0>
    }}
   onGround_Crouch_Moving {simSpaceSettings {
     maxAngularAcceleration=1000
     maxAngularVelocity=3
     maxLinearAcceleration=300
     maxLinearVelocity=700
     velocityScaleZ=0
     externalLinearDragV=<0 0.5 0>
    }}
   onGround_Falling {simSpaceSettings {
     maxAngularAcceleration=1000
     maxAngularVelocity=3
     maxLinearAcceleration=300
     maxLinearVelocity=700
     velocityScaleZ=0
     externalLinearDragV=<0 0.5 0>
    }}
   onGround_Moving {
    componentLinearAccScale=<0.3 0.3 0.3>
    componentLinearVelScale=<0.3 0.3 0.3>
    simSpaceSettings {
     maxAngularAcceleration=1000
     maxAngularVelocity=3
     maxLinearAcceleration=300
     maxLinearVelocity=700
     velocityScaleZ=0
     externalLinearDragV=<0 0.5 0>
    }
   }
   onGround_Standing {
    componentLinearAccScale=<0.3 0.4 0.4>
    componentLinearVelScale=<0.3 0.4 0.4>
    simSpaceSettings {
     maxAngularAcceleration=1000
     maxAngularVelocity=3
     maxLinearAcceleration=300
     maxLinearVelocity=700
     velocityScaleZ=0
     externalLinearDragV=<0 0.5 0>
    }
   }
   rail_Grinding {simSpaceSettings {
     maxAngularAcceleration=1000
     maxAngularVelocity=3
     maxLinearAcceleration=300
     maxLinearVelocity=700
     velocityScaleZ=0
     worldAlpha=0.1
     externalLinearDragV=<0 0.5 0>
    }}
   riding_Moving {simSpaceSettings {
     maxAngularAcceleration=1000
     maxAngularVelocity=3
     maxLinearAcceleration=300
     maxLinearVelocity=700
     velocityScaleZ=0
     externalLinearDragV=<0 0.5 0>
    }}
   riding_Moving_Targeting {simSpaceSettings {
     maxAngularAcceleration=1000
     maxAngularVelocity=3
     maxLinearAcceleration=300
     maxLinearVelocity=700
     velocityScaleZ=0
     externalLinearDragV=<0 0.5 0>
    }}
   riding_Standing {simSpaceSettings {
     maxAngularAcceleration=1000
     maxAngularVelocity=3
     maxLinearAcceleration=300
     maxLinearVelocity=700
     velocityScaleZ=0
     externalLinearDragV=<0 0.5 0>
    }}
   skydive_DivingDown {simSpaceSettings {
     maxAngularAcceleration=1000
     maxAngularVelocity=3
     maxLinearAcceleration=300
     maxLinearVelocity=700
     velocityScaleZ=0
     externalLinearDragV=<0 0.5 0>
    }}
   skydive_FloatingDown {
    componentLinearAccScale=<0 0.5 0>
    componentLinearVelScale=<0 0.5 0>
    simSpaceSettings {
     maxAngularAcceleration=1000
     maxAngularVelocity=3
     maxLinearAcceleration=300
     maxLinearVelocity=700
     velocityScaleZ=0
     externalLinearDragV=<0 0.5 0>
    }
   }
   skydive_GoingUp {
    componentLinearAccScale=<0.4 0.4 0.4>
    componentLinearVelScale=<0.4 0.4 0.4>
    simSpaceSettings {
     maxAngularAcceleration=1000
     maxAngularVelocity=3
     maxLinearAcceleration=300
     maxLinearVelocity=700
     velocityScaleZ=0
     externalLinearDragV=<0 0.5 0>
    }
   }
   skydive_Parachuting {
    componentLinearAccScale=<0 0.1 0.3>
    componentLinearVelScale=<0 0.1 0.3>
    simSpaceSettings {
     maxAngularAcceleration=1000
     maxAngularVelocity=3
     maxLinearAcceleration=300
     maxLinearVelocity=700
     velocityScaleZ=0
     externalLinearDragV=<0 0.5 0>
    }
   }
   skydive_Surfing {simSpaceSettings {
     maxAngularAcceleration=1000
     maxAngularVelocity=3
     maxLinearAcceleration=300
     maxLinearVelocity=700
     velocityScaleZ=0
     externalLinearDragV=<0 0.5 0>
    }}
   swimming {simSpaceSettings {
     maxAngularAcceleration=1000
     maxAngularVelocity=3
     maxLinearAcceleration=300
     maxLinearVelocity=700
     velocityScaleZ=0
     externalLinearDragV=<0 0.5 0>
    }}
   tactical_Sprint {simSpaceSettings {
     maxAngularAcceleration=1000
     maxAngularVelocity=3
     maxLinearAcceleration=300
     maxLinearVelocity=700
     velocityScaleZ=0
     externalLinearDragV=<0 0.5 0>
    }}
   ziplining {simSpaceSettings {
     maxAngularAcceleration=1000
     maxAngularVelocity=3
     maxLinearAcceleration=300
     maxLinearVelocity=700
     velocityScaleZ=0
     externalLinearDragV=<0 0.5 0>
    }}
  }
 }
 RBAN Settings Out {
  type=RBANSettingStruct
  default {
   componentLinearAccScale=<0 0 0>
   componentLinearVelScale=<0 0 0>
   simSpaceSettings {
    dampingAlpha=1
    maxAngularAcceleration=1000
    maxAngularVelocity=3
    maxLinearAcceleration=300
    maxLinearVelocity=700
    velocityScaleZ=0
    worldAlpha=1
    externalAngularVelocity=<0 0 0>
    externalLinearDragV=<0 0.5 0>
    externalLinearVelocity=<0 0 0>
   }
  }
 }
 Sim Space Settings {
  type=SimSpaceSettings
  default {
   dampingAlpha=1
   maxAngularAcceleration=500
   maxAngularVelocity=3
   maxLinearAcceleration=400
   maxLinearVelocity=250
   velocityScaleZ=0
   worldAlpha=1
   externalAngularVelocity=<0.4 0.4 0.4>
   externalLinearDragV=<0.1 0.1 0>
   externalLinearVelocity=<0.4 0.4 0.4>
  }
 }
 Sky_aplha {type=number}
]
