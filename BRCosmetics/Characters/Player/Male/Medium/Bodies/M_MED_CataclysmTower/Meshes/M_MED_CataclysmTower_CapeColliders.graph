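# Physics asset of cape colliders for the male medium CataclysmTower player cosmetic body.
# Capsules cover the pelvis, spine, arms, hands, thighs, calves and feet, forming an approximate body volume.
# Smaller capsules sit on three left and three right back-jacket dynamic bones.
# It defines no constraints, so it holds only collision shapes, with no joint setup between bodies.
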PhysicsAsset /BRCosmetics/Characters/Player/Male/Medium/Bodies/M_MED_CataclysmTower/Meshes/M_MED_CataclysmTower_CapeColliders
bodies {
 Dyn_jacket_Back_L_01 [{shapeType=Capsule, length=5.7331, radius=7.2085, shapeName=Dyn_jacket_Back_L_01_capsule, center=<0 0 -2.3712>, rotation=<rot 0 0 90>}]
 Dyn_jacket_Back_L_02 [{shapeType=Capsule, length=5.7331, radius=7.2085, shapeName=Dyn_jacket_Back_L_02_capsule, center=<0 0 -2.3712>, rotation=<rot 0 0 90>}]
 Dyn_jacket_Back_L_03 [{shapeType=Capsule, length=5.7331, radius=7.2085, shapeName=Dyn_jacket_Back_L_03_capsule, center=<0 0 -2.3712>, rotation=<rot 0 0 90>}]
 Dyn_jacket_Back_R_01 [{
   shapeType=Capsule
   length=5.7331
   radius=7.2085
   shapeName=Dyn_jacket_Back_R_01_capsule
   center=<0.0669 0.1482 2.1776>
   rotation=<rot 0 0 90>
  }]
 Dyn_jacket_Back_R_02 [{
   shapeType=Capsule
   length=5.7331
   radius=7.2085
   shapeName=Dyn_jacket_Back_R_02_capsule
   center=<0.0669 0.1482 2.1776>
   rotation=<rot 0 0 90>
  }]
 Dyn_jacket_Back_R_03 [{
   shapeType=Capsule
   length=5.7331
   radius=7.2085
   shapeName=Dyn_jacket_Back_R_03_capsule
   center=<0.0669 0.1482 2.1776>
   rotation=<rot 0 0 90>
  }]
 calf_l [{shapeType=Capsule, length=32.5143, radius=15.6454, center=<-23.0006 -2.2427 0>, rotation=<rot 90 0 0>}]
 calf_r [{shapeType=Capsule, length=32.5143, radius=15.6454, center=<23.0006 2.2427 0>, rotation=<rot 90 0 180>}]
 foot_l [{shapeType=Capsule, length=10.5461, radius=14.1734, center=<-4.1955 6.0815 0>, rotation=<rot 0 14.9998 90>}]
 foot_r [{shapeType=Capsule, length=10.5461, radius=14.1734, center=<4.1955 -6.0815 0>, rotation=<rot 0 14.9998 -90>}]
 hand_l [{shapeType=Capsule, length=9.5015, radius=6.7933, center=<6.2703 0 0>, rotation=<rot 90 0 0>}]
 hand_r [{shapeType=Capsule, length=9.5015, radius=6.7933, center=<-6.2703 0 0>, rotation=<rot 90 0 180>}]
 lowerarm_l [{shapeType=Capsule, length=14.3387, radius=7.5574, center=<8.9852 0 0.2267>, rotation=<rot 90 0 0>}]
 lowerarm_r [{shapeType=Capsule, length=14.3387, radius=7.5574, center=<-8.9852 0 -0.2267>, rotation=<rot 90 0 180>}]
 pelvis [{shapeType=Capsule, length=6.7241, radius=16.0874, center=<-2.4855 -0.9262 0>}]
 spine_02 [{shapeType=Capsule, length=3.3213, radius=14.7784, center=<-2.6142 0 0>}]
 spine_03 [{shapeType=Capsule, length=4.9972, radius=14.7226}]
 spine_04 [{shapeType=Capsule, length=4.2255, radius=15.6781, center=<0 1.273 0>}]
 spine_05 [{shapeType=Capsule, length=14.9226, radius=14.9992, center=<0 1.0616 0>}]
 thigh_l [{shapeType=Capsule, length=15.4697, radius=14.5854, center=<-18.4217 -1.038 0>, rotation=<rot 90 180 2.664>}]
 thigh_r [{shapeType=Capsule, length=16.4633, radius=13.5918, center=<18.4217 1.038 0>, rotation=<rot 90 180 -177.336>}]
 upperarm_l [{shapeType=Capsule, length=17.1712, radius=7.7501, center=<10.0455 0 0>, rotation=<rot 90 0 0>}]
 upperarm_r [{shapeType=Capsule, length=17.1712, radius=7.7501, center=<-10.0455 0 0>, rotation=<rot 90 0 180>}]
}
constraints []
