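# Control Rig for the CataclysmTower male medium body's face accessory part that adjusts the back hair bones for skydive and crouch-like poses.
# A branch on isSkydive applies fixed full-weight rotations to the Base_Hair_Back bones when true.
# Otherwise it rotates Hair_Back_02, with weight remapped from pelvis height relative to root so the offset grows as the pelvis lowers.
# The isCrouch variable is declared but unused in the graph.
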ControlRigBlueprint /BRCosmetics/Characters/Player/Male/Medium/Bodies/M_MED_CataclysmTower/Meshes/Parts/M_MED_CataclysmTower_FaceAcc_CtrlRig
graphs [
 RigVMFunctionLibrary {links [], nodes []}
 RigVMModel {nodes [
   RigUnit_BeginExecution = &RigUnit_BeginExecution()
   n1 = &RigVMFunction_ControlFlowBranch(ExecuteContext=@RigUnit_BeginExecution.ExecutePin, Condition=RigVMVariableNode() {variable=isSkydive}.Value)
   &RigUnit_ModifyTransforms(
    ItemToModify=((Item=(Type=Bone,Name="Hair_Back_02"),Transform=(Rotation=(X=0,Y=0.5042,Z=0,W=0.8636),Translation=(X=0,Y=0,Z=0),Scale3D=(X=1,Y=1,Z=1))))
    WeightMinimum=0
    Weight=&RigVMFunction_MathDoubleRemap(
     SourceMaximum=39
     SourceMinimum=70
     TargetMaximum=1
     TargetMinimum=0
     Value=&RigUnit_GetRelativeTransformForItem(Child=(Type=Bone,Name="pelvis"), Parent=(Type=Bone,Name="root"), bChildInitial=False, bParentInitial=False).RelativeTransform.Translation.Z
    ).Result
    ExecutePin=@n1.False
   )
   &RigUnit_ModifyTransforms(
    ItemToModify="((Item=(Type=Bone,Name=\"Base_Hair_Back_02\"),Transform=(Rotation=(X=0,Y=0.2383,Z=0,W=0.9712),Translation=(X=0,Y=0,Z=0),Scale3D=(X=1,Y=1,Z=1))),(Item=(Type=Bone,Name=\"Base_Hair_Back_01\"),Transform=(Rotation=(X=0,Y=0.2626,Z=0,W=0.9649),Translation=(X=0,Y=0,Z= ...(truncated)"
    Weight=1
    WeightMinimum=0
    ExecutePin=@n1.True
   )
  ]}
]
hierarchy [
 root {loc=<0 0 0>, rot=<0 0 0 1>}
 pelvis {parent=root, loc=<0 -0.7071 93.1191>, rot=<0.0127 0.707 -0.0127 -0.707>}
 spine_01 {parent=pelvis, loc=<6.7494 -0.6388 0>, rot=<0 0 0.034 -0.9994>}
 spine_02 {parent=spine_01, loc=<7.252 -0.035 0>, rot=<0 0 0.0238 -0.9997>}
 spine_03 {parent=spine_02, loc=<6.0106 0 0>, rot=<0 0 0.0149 -0.9999>}
 spine_04 {parent=spine_03, loc=<8.1023 0 0>, rot=<0 0 -0.0132 -0.9999>}
 spine_05 {parent=spine_04, loc=<10.0142 0 0>, rot=<0 0 -0.0015 -1>}
 neck_01 {parent=spine_05, loc=<11.3138 0.1529 0>, rot=<0 0 -0.2003 -0.9797>}
 neck_02 {parent=neck_01, loc=<4.5627 0 0>, rot=<0 0 0.0504 -0.9987>}
 head {parent=neck_02, loc=<5.3914 0 0>, rot=<0 0 0.115 -0.9934>}
 Hair_L_01 {parent=head, loc=<7.2852 9.9227 -4.3957>, rot=<0.1569 -0.8605 -0.4773 -0.0851>}
 Hair_L_02 {parent=head, loc=<7.9195 4.871 -7.6805>, rot=<0.1108 -0.9277 -0.1171 -0.3367>}
 Hair_L_03 {parent=head, loc=<6.6275 -3.4037 -8.2751>, rot=<0.1246 -0.6816 0.3851 -0.6096>}
 Hair_R_01 {parent=head, loc=<5.6227 -1.5603 8.9236>, rot=<0.4821 -0.3695 0.794 0.0271>}
 Hair_R_02 {parent=head, loc=<5.9418 6.9135 7.4325>, rot=<0.3777 0.2334 0.8707 -0.2117>}
 Hair_L_04 {parent=head, loc=<9.1629 -2.7231 -8.1289>, rot=<0.2017 -0.7419 0.2061 -0.6053>}
 Hair_R_03 {parent=head, loc=<8.4491 11.522 3.45>, rot=<0.0839 0.247 0.9445 -0.1995>}
 Hair_R_04 {parent=head, loc=<7.6567 -3.341 7.8986>, rot=<0.6023 -0.2283 0.7477 -0.1615>}
 Base_Hair_Back_02 {parent=head, loc=<2.6314 -6.2475 5.0294>, rot=<0.3435 -0.5658 0.748 -0.0496>}
 Hair_Back_02 {parent=Base_Hair_Back_02, loc=<0 0 0>, rot=<0 0 0 1>}
 Base_Hair_Back_01 {parent=head, loc=<2.4505 -7.2668 -1.1232>, rot=<0.1829 -0.6744 0.6714 -0.2471>}
 Hair_Back_01 {parent=Base_Hair_Back_01, loc=<0 0 0>, rot=<0 0 0 1>}
]
variables [
 isCrouch {type=double, default=0}
 isSkydive {type=bool, default=False}
]
