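# Physics asset for the face accessory mesh of the CataclysmTower medium male player body, covering its hair bones and head.
# Seven small, identically sized capsules sit on the back, left and right hair bones, with one larger capsule on the head.
# Each hair bone is constrained directly to the head rather than chained, with a narrow, uniform swing limit and a zero twist limit.
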PhysicsAsset /BRCosmetics/Characters/Player/Male/Medium/Bodies/M_MED_CataclysmTower/Meshes/Parts/M_MED_CataclysmTower_FaceAcc_Physics
bodies {
 Hair_Back_01 [{shapeType=Capsule, length=3, radius=1.5, shapeName=Hair_Back_01_capsule, center=<1.8422 0 0>, rotation=<rot 0 0 90>}]
 Hair_Back_02 [{shapeType=Capsule, length=3, radius=1.5, shapeName=Hair_Back_02_capsule, center=<1.8422 0 0>, rotation=<rot 0 0 90>}]
 Hair_L_01 [{shapeType=Capsule, length=3, radius=1.5, shapeName=Hair_L_01_capsule, center=<1.8422 0 0>, rotation=<rot 90 0 0>}]
 Hair_L_02 [{shapeType=Capsule, length=3, radius=1.5, shapeName=Hair_L_02_capsule, center=<1.8422 0 0>, rotation=<rot 90 0 0>}]
 Hair_L_03 [{shapeType=Capsule, length=3, radius=1.5, shapeName=Hair_L_03_capsule, center=<1.8422 0 0>, rotation=<rot 90 0 0>}]
 Hair_R_01 [{shapeType=Capsule, length=3, radius=1.5, shapeName=Hair_R_01_capsule, center=<1.8422 0 0>, rotation=<rot 90 0 0>}]
 Hair_R_02 [{shapeType=Capsule, length=3, radius=1.5, shapeName=Hair_R_02_capsule, center=<1.8422 0 0>, rotation=<rot 90 0 0>}]
 head [{shapeType=Capsule, length=3.4802, radius=6.5198, shapeName=head_capsule, rotation=<rot -90 -2.7321 2.7316>}]
}
constraints [
 {bone1Name=Hair_Back_01, bone2Name=head, swing1LimitDegrees=20, swing2LimitDegrees=20, twistLimitDegrees=0}
 {bone1Name=Hair_Back_02, bone2Name=head, swing1LimitDegrees=20, swing2LimitDegrees=20, twistLimitDegrees=0}
 {bone1Name=Hair_L_01, bone2Name=head, swing1LimitDegrees=20, swing2LimitDegrees=20, twistLimitDegrees=0}
 {bone1Name=Hair_L_02, bone2Name=head, swing1LimitDegrees=20, swing2LimitDegrees=20, twistLimitDegrees=0}
 {bone1Name=Hair_L_03, bone2Name=head, swing1LimitDegrees=20, swing2LimitDegrees=20, twistLimitDegrees=0}
 {bone1Name=Hair_R_01, bone2Name=head, swing1LimitDegrees=20, swing2LimitDegrees=20, twistLimitDegrees=0}
 {bone1Name=Hair_R_02, bone2Name=head, swing1LimitDegrees=20, swing2LimitDegrees=20, twistLimitDegrees=0}
]
